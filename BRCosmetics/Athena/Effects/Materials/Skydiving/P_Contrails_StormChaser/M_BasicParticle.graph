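# Unlit translucent particle material for the StormChaser skydiving contrail effects, usable on sprites and Niagara ribbons.
# An inky smoke trail texture's green channel, times particle alpha, drives depth-faded opacity; UseRedChannel has no effect because both lerp inputs read green.
# That mask also blends squared particle color (via SaturationBlend) toward plain particle color, which Brightness scales before MF_FX_EmissiveAdjustments.
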Material /BRCosmetics/Athena/Effects/Materials/Skydiving/P_Contrails_StormChaser/M_BasicParticle
graph {
 properties {
  bUsedWithNiagaraRibbons=true
  bUsedWithNiagaraSprites=true
  bUsedWithParticleSprites=true
  blendMode=BLEND_Translucent
  pixelDepthOffsetMode=PDOM_Legacy
  shadingModel=MSM_Unlit
  usedShadingModels=Unlit
 }
 nodes [
  n1 = ParticleColor()
  n2 = TextureSampleParameter2D() {parameterName=BaseTexture, texture=/BRCosmetics/Athena/Effects/Textures/T_InkySmokeTrail}
  n3 = LinearInterpolate(A=@n2.G, B=@n2.G, Alpha=ScalarParameter() {parameterName=UseRedChannel})
  DepthFade(Opacity=Multiply(A=@n1.A, B=@n3), FadeDistance=ScalarParameter() {defaultValue=20, parameterName=FadeDist})
  n4 = LinearInterpolate(A=@n1.RGB, B=Multiply(A=@n1.RGB, B=@n1.RGB), Alpha=ScalarParameter() {parameterName=SaturationBlend})
  MaterialFunctionCall(Emissive=Multiply(A=LinearInterpolate(A=@n4, B=@n1.RGB, Alpha=@n3), B=ScalarParameter() {defaultValue=1, parameterName=Brightness})) {materialFunction=&MF_FX_EmissiveAdjustments}
 ]
}
